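# Animation Blueprint for the Traction gadget prop skeleton that steers its front wheels as the owning actor turns.
# Each update, Rotate Wheels scales the actor's yaw change since the previous frame, clamps it and interpolates the wheel angle toward it.
# The AnimGraph switches between InMotion and DefaultSlot montage slots on bIsMoving2D, then additively yaws both front wheel bones by that angle.
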AnimBlueprint /Game/Animation/Props/Gadgets_Traction/Meshes/Traction_Skeleton_AnimBlueprint
target_skeleton=&/Traction_Gadget_Skeleton
graphs [
 AnimGraph {nodes [
   n1 = Slot 'DefaultSlot'()
   n2 = Slot 'InMotion'() {slotName=InMotion}
   n3 = Get bIsMoving2D()
   n4 = Get Wheel Rotation() -> (Wheel Rotation:double)
   n5 = "Transform (Modify) Bone - Bone: JNT_r_Front_Wheel"!(
    ComponentPose=Local To Component(LocalPose=Blend Poses by bool(BlendPose_0=@n2, BlendPose_1=@n1, bActiveValue=@n3) {blendType=Linear})
    Alpha=1
    Translation="0,0,0"
    Rotation_Roll:float=0
    Rotation_Pitch:float=0
    Rotation_Yaw:float=@n4
    Scale="1,1,1"
   ) -> (Pose) {rotationMode=BMM_Additive, alphaBoolBlend {blendOption=Linear}, boneToModify {boneName=JNT_r_Front_Wheel}}
   Output Pose(Result=Component To Local(ComponentPose="Transform (Modify) Bone - Bone: JNT_l_Front_Wheel"!(ComponentPose=@n5, Alpha=1, Translation="0,0,0", Rotation_Roll:float=0, Rotation_Pitch:float=0, Rotation_Yaw:float=@n4, Scale="1,1,1") -> (Pose) {rotationMode=BMM_Additive, alphaBoolBlend {blendOption=Linear}, boneToModify {boneName=JNT_l_Front_Wheel}})) {layerGroup=None}
  ]}
 EventGraph {nodes [
   n1 = Event BlueprintUpdateAnimation()
   > n2 = Rotate Wheels!(
    execute
    self
    OwningActor:Actor Object Reference=Get OwningActor() -> (OwningActor:Actor Object Reference)
    CurrentWheelRotation:double=Get Wheel Rotation()
    DeltaTime:double=@n1.DeltaTimeX
    CurrentActorRotation:double=Get Actor Rotation() -> (Actor Rotation:double)
   ) -> (then, NewActorRotation:double, NewWheelRotation:double)
   > Set Actor Rotation(Actor Rotation:double=@n2) -> (Output_Get:double)
   > Set Wheel Rotation(Wheel Rotation:double=@n2.NewWheelRotation) -> (Output_Get:double)
   Event BlueprintInitializeAnimation()
   > Set OwningActor(OwningActor:Actor Object Reference=GetOwningActor()) -> (Output_Get:Actor Object Reference)
  ]}
 Rotate Wheels {nodes [
   n1 = Rotate Wheels!() -> (then, OwningActor:Actor Object Reference, CurrentWheelRotation:double=~, DeltaTime:double=~, CurrentActorRotation:double=~) {is=title K2Node_FunctionEntry}
   n2 = Get Actor Rotation!(self:Actor Object Reference=@n1) -> (ReturnValue_Roll:float, ReturnValue_Pitch:float, ReturnValue_Yaw:float) {is=title K2Node_CallFunction}
   n3 = float * float(A=float - float(A=@n2.ReturnValue_Yaw, B=@n1.CurrentActorRotation) {is=title K2Node_CallFunction}, B=5)
   > Return Node!(
    execute
    NewActorRotation:double=@n2.ReturnValue_Yaw
    NewWheelRotation:double=FInterpTo(Current=@n1.CurrentWheelRotation, Target="Clamp (Float)"(Value=@n3, Min=-45, Max=45), DeltaTime=@n1.DeltaTime, InterpSpeed=5)
   )
  ]}
]
variables [
 Actor Rotation {type=number}
 OwningActor {type=Actor, default=null}
 Wheel Rotation {type=number}
]
